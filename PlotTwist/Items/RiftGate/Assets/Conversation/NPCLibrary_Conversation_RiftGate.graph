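# Conversation database for the RiftGate item in PlotTwist, named as an NPC library conversation.
# It registers a single entry point under the tag Conversation.Entry.PlotTwist.Force.Projectile, mapped to its one entry node.
# That node has no outputs, so the graph holds no dialogue, choices or tasks beyond the entry itself.
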ConversationDatabase /PlotTwist/Items/RiftGate/Assets/Conversation/NPCLibrary_Conversation_RiftGate
entry_points [{tag=Conversation.Entry.PlotTwist.Force.Projectile, destinations [#1]}]
nodes [ConversationEntryPointNode_0 {
  id=#1
  class=ConversationEntryPointNode
  outputs []
  properties {entryTag {tagName=Conversation.Entry.PlotTwist.Force.Projectile}}
 }]
